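# Niagara system for the Spotlight emote, adding warm light and glow that follow a socket or bone on the lamp shade skeletal mesh.
# Lightcone renders two particles as cone meshes and glow sprites, sized per partition.
# LightEpic uses a component renderer, while LightHigh uses two offset light-renderer particles with specular zeroed in the front end.
# All particles fade in and persist past lifetime; random-alpha flicker modules are disabled.
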NiagaraSystem /BRCosmetics/Animation/Game/MainPlayer/Emotes/Spotlight/FX/NS_Emote_Spotlight
dependencies {
 usedDataInterfaces [
  &NiagaraDataInterfaceCurve
  &NiagaraDataInterfaceArrayFloat3
  &NiagaraDataInterfaceArrayFloat
  &NiagaraDataInterfaceSkeletalMesh
 ]
 usedDynamicInputs []
 usedModules [
  &self:SystemSpawnScript.SystemScriptSource.SystemScriptGraph.NiagaraNodeAssignment_0.SetVariables_#1
  &SystemState
 ]
}
emitters {
 LightEpic {
  emitterUpdateScript [
   &EmitterState(Life Cycle Mode=::NewEnumerator1, Loop Behavior=::NewEnumerator1, Loop Duration=1, UseLoopDelay=false)
   &SpawnBurst_Instantaneous(Use Spawn Probability=false, UseLoopCountLimit=false)
  ]
  particleSpawnScript [&InitializeParticle(Color=<rgba 1 0.6017 0.0288 1>, UsePositionOffset=false, Write Material Random=true)]
  particleUpdateScript [
   &ParticleState(Kill Particles When Lifetime Has Elapsed=false)
   &ScaleColor(
    Scale Alpha=&FloatFromCurve(FloatCurve=DataInterface() {propertyValues {Curve {keys [<at 0 0>,<at 0.1 1>,<at 1 1>]}}})
    ScaleA=true
    ScaleRGB=true
   )
   &ScaleColor(
    Scale Alpha=&RandomRangeFloat(Minimum=0.8, Maximum=1, Evaluation Type=::NewEnumerator1)
    ScaleA=true
    ScaleRGB=true
   ) {enabled=false, module=ScaleColor001}
   &SkeletalMeshLocation(
    Bone / Socket Index=1
    Bone Sampling Mode=::NewEnumerator4
    MaskSpawnGroup=false
    Sampled Position Offset=<0 0 25>
    Skeletal Mesh=$System.LampShadeSKM
   )
   &FN_Light_Attributes(Color=$Particles.Color, Create Color Attribute=true, Create Exponent Attribute=true, Create Radius Attribute=true, Exponent=1, Radius=300) {module=FN_Light_Attributes001}
  ]
  rendererClasses [&NiagaraComponentRendererProperties]
  renderers [{rendererClass=&NiagaraComponentRendererProperties, rendererIndex=0}]
 }
 LightHigh {
  emitterUpdateScript [
   &EmitterState(Life Cycle Mode=::NewEnumerator1, Loop Behavior=::NewEnumerator1, Loop Duration=1, UseLoopDelay=false)
   &SpawnBurst_Instantaneous(Spawn Count=2, Use Spawn Probability=false, UseLoopCountLimit=false)
  ]
  particleSpawnScript [
   &PartitionParticles()
   &InitializeParticle(Color=<rgba 1 0.6017 0.0288 1>, UsePositionOffset=false, Write Material Random=true)
  ]
  particleUpdateScript [
   &ParticleState(Kill Particles When Lifetime Has Elapsed=false)
   &ScaleColor(
    Scale Alpha=&FloatFromCurve(FloatCurve=DataInterface() {propertyValues {Curve {keys [<at 0 0>,<at 0.1 1>,<at 1 1>]}}})
    ScaleA=true
    ScaleRGB=true
   )
   &ScaleColor(
    Scale Alpha=&RandomRangeFloat(Minimum=0.8, Maximum=1, Evaluation Type=::NewEnumerator1)
    ScaleA=true
    ScaleRGB=true
   ) {enabled=false, module=ScaleColor001}
   &SkeletalMeshLocation(
    Bone / Socket Index=1
    Bone Sampling Mode=::NewEnumerator4
    MaskSpawnGroup=false
    Sampled Position Offset=&SelectVectorFromArray(
     Direct Array Index=$Particles.PartitionParticles.Partition
     Vector Selection Array=DataInterface() {propertyValues {FloatData [<0 0 85>, <0 0 -125>]}}
    )
    Skeletal Mesh=$System.LampShadeSKM
   )
   &FN_Light_Attributes(
    Color=$Particles.Color
    Create Color Attribute=true
    Create Exponent Attribute=true
    Create Radius Attribute=true
    Exponent=1
    Radius=&SelectFloatFromArray(Direct Array Index=$Particles.PartitionParticles.Partition, Float Selection Array=DataInterface() {propertyValues {FloatData [100,130]}})
    Scale Color=5
   )
   /SetVariables(Particles.LightSpecularScale:NiagaraFloat=&FN_FrontEndFloatModifier(Gameplay=1, Front End=0)) {isSetParameters=true}
  ]
  rendererClasses [&NiagaraLightRendererProperties]
  renderers [{rendererClass=&NiagaraLightRendererProperties, rendererIndex=0}]
 }
 Lightcone {
  emitterUpdateScript [
   &EmitterState(Life Cycle Mode=::NewEnumerator1, Loop Behavior=::NewEnumerator1, Loop Duration=1, UseLoopDelay=false)
   &SpawnBurst_Instantaneous(Spawn Count=2, Use Spawn Probability=false, UseLoopCountLimit=false)
  ]
  particleSpawnScript [
   &PartitionParticles()
   &InitializeParticle(
    Color=<rgba 1 0.5971 0.1062 0.14>
    Lifetime=0.5
    Mesh Scale=&MaskVectorBySpawnGroup(Masked Vector=<1 1 0.7>, Passthrough Vector=<1.7 1.7 -0.5>, Spawn Group Mask=$Particles.PartitionParticles.Partition)
    Mesh Scale Mode=::NewEnumerator4
    Sprite Size Mode=::NewEnumerator3
    Uniform Sprite Size=&MaskFloatBySpawnGroup(Masked Float=55, Passthrough Float=85, Spawn Group Mask=$Particles.PartitionParticles.Partition)
    UsePositionOffset=false
    Write Material Random=true
   )
  ]
  particleUpdateScript [
   &ParticleState(Kill Particles When Lifetime Has Elapsed=false)
   &SkeletalMeshLocation(
    Bone / Socket Index=1
    Bone Sampling Mode=::NewEnumerator4
    MaskSpawnGroup=false
    Sampled Position Offset=&MaskVectorBySpawnGroup(Masked Vector=<0 0 25>, Passthrough Vector=<0 0 0>, Spawn Group Mask=$Particles.PartitionParticles.Partition)
    Skeletal Mesh=$System.LampShadeSKM
   )
   &ScaleColor(
    Scale Alpha=&FloatFromCurve(FloatCurve=DataInterface() {propertyValues {Curve {keys [<at 0 0>,<at 0.1 1>,<at 1 1>]}}})
    ScaleA=true
    ScaleRGB=true
   )
   &ScaleColor(
    Scale Alpha=&RandomRangeFloat(Minimum=0.5, Maximum=1, Evaluation Type=::NewEnumerator1)
    ScaleA=true
    ScaleRGB=true
   ) {enabled=false, module=ScaleColor001}
   /SetVariables(Particles.GlowCameraOffset=23.3, Particles.GlowColor=<rgba 1 0.6382 0 0.3>) {isSetParameters=true}
  ]
  rendererClasses [&NiagaraMeshRendererProperties, &NiagaraSpriteRendererProperties]
  renderers [
   {rendererClass=&NiagaraMeshRendererProperties, rendererIndex=0}
   {rendererClass=&NiagaraSpriteRendererProperties, rendererIndex=1}
  ]
 }
}
summary {
 systemName=NS_Emote_Spotlight
 emitters [
  {emitterName=Lightcone, rendererClasses [&NiagaraMeshRendererProperties, &NiagaraSpriteRendererProperties]}
  {emitterName=LightEpic, rendererClasses [&NiagaraComponentRendererProperties]}
  {emitterName=LightHigh, rendererClasses [&NiagaraLightRendererProperties]}
 ]
 userVariables [
  User.bIsFrontEnd {type=NiagaraBool}
  User.bIsFrontEndPreview {type=NiagaraBool}
 ]
}
