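# Niagara system for the moth effect in the Spotlight emote, with a soft blue glow and sparkling dust placed on the lamp shade skeletal mesh.
# Dust steadily spawns short-lived cyan sprites around sampled mesh points, which fall under gravity as their size pops and shrinks.
# It has two sprite renderers and sets a separate blue glow color and enlarged glow size.
# Glow burst-spawns a sprite and re-samples the mesh every update to follow it.
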NiagaraSystem /BRCosmetics/Animation/Game/MainPlayer/Emotes/Spotlight/FX/NS_Emote_Spotlight_Moth
dependencies {
 usedDataInterfaces [&NiagaraDataInterfaceCurve, &NiagaraDataInterfaceSkeletalMesh]
 usedDynamicInputs []
 usedModules [
  &self:SystemSpawnScript.SystemScriptSource.SystemScriptGraph.NiagaraNodeAssignment_0.SetVariables_#1
  &SystemState
 ]
}
emitters {
 Dust {
  emitterUpdateScript [
   &EmitterState()
   &SpawnRate(SpawnRate=12, Use Spawn Probability=false)
  ]
  particleSpawnScript [
   &InitializeParticle(
    Color=<rgba 0.8991 6 5.4017 1>
    Lifetime Max=0.6
    Lifetime Min=0.2
    Lifetime Mode=::NewEnumerator1
    Sprite Size Mode=::NewEnumerator1
    Uniform Sprite Size Max=2
    Uniform Sprite Size Min=1
    UsePositionOffset=false
    Write Material Random=true
   )
   &SkeletalMeshLocation(Bone Sampling Mode=::NewEnumerator4, MaskSpawnGroup=false, Skeletal Mesh=$System.LampShadeSKM)
   &ShapeLocation(MaskSpawnGroup=false, Override Seed=false, Sphere Radius=6, UseOwnerScale=true)
  ]
  particleUpdateScript [
   &ParticleState()
   &GravityForce(Gravity=<0 0 -200>)
   &SolveForcesAndVelocity(Clamp Velocity=false, Limit Acceleration=false)
   &ScaleSpriteSize(Uniform Curve Sprite Scale=DataInterface() {propertyValues {Curve {keys [
        {interpMode=RCIM_Cubic, leaveTangent=6.0949, tangentMode=RCTM_Break, time=0, value=0.4}
        <at 0.2 1>
        {interpMode=RCIM_Cubic, leaveTangent=-3.2625, tangentMode=RCTM_Break, time=0.3, value=1}
        {tangentMode=RCTM_User, time=1, value=0}
       ]}}})
   /SetVariables(
    Particles.GlowColor=<rgba 0.1455 0.3732 1 0.5>
    Particles.GlowSize=&Multiply_Vector2D_ByFloat(Float=5, Vector2D=$Particles.SpriteSize)
   ) {isSetParameters=true}
  ]
  rendererClasses [&NiagaraSpriteRendererProperties]
  renderers [
   {rendererClass=&NiagaraSpriteRendererProperties}
   {rendererClass=&NiagaraSpriteRendererProperties, rendererIndex=1}
  ]
 }
 Glow {
  emitterUpdateScript [
   &EmitterState(Life Cycle Mode=::NewEnumerator1, Loop Duration=1, UseLoopDelay=false)
   &SpawnBurst_Instantaneous(Use Spawn Probability=false, UseLoopCountLimit=false)
  ]
  particleSpawnScript [&InitializeParticle(
    Color=<rgba 0.1455 0.3732 1 0.5>
    Sprite Size Mode=::NewEnumerator3
    Uniform Sprite Size=15
    UsePositionOffset=false
    Write Material Random=true
   )]
  particleUpdateScript [
   &ParticleState()
   &SkeletalMeshLocation(Bone Sampling Mode=::NewEnumerator4, MaskSpawnGroup=false, Skeletal Mesh=$System.LampShadeSKM)
  ]
  rendererClasses [&NiagaraSpriteRendererProperties]
  renderers [{rendererClass=&NiagaraSpriteRendererProperties}]
 }
}
summary {
 systemName=NS_Emote_Spotlight_Moth
 emitters [
  {emitterName=Dust, rendererClasses [&NiagaraSpriteRendererProperties]}
  {emitterName=Glow, rendererClasses [&NiagaraSpriteRendererProperties]}
 ]
 userVariables []
}
